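; Scriptable temple fire pot that can be lit or put out and applies Curie fire interactions to overlapping actors.
; TorchState sets Enabled, whose OnRep toggles Niagara fire, audio, light radius and collision volume; lighting it also checks actors already inside.
; On authority, overlaps begin fire interactions and store handles per actor in a map, ending them on exit.
; Destruction ends all stored interactions so nothing burns forever; death fades out audio.
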(event EventBeginPlay
  ;[1 1 1] Init
  (Variables|Default|SetEnabled (Variables|Default|GetEnabled)))

(event EventOnDeathPlayEffects
  (params (Damage :type Float%20%28single-precision%29) (DamageTags :type Gameplay%20Tag%20Container%20Structure%20%28by%20ref%29) (Momentum :type Vector :default "0, 0, 0") (HitInfo :type Hit%20Result%20Structure%20%28by%20ref%29) (InstigatedBy :type Fort%20Pawn%20Object%20Reference) (DamageCauser :type Actor%20Object%20Reference) (EffectContext :type Gameplay%20Effect%20Context%20Handle%20Structure))
  ;[1 1 1] On Death Play Effects
  (Utilities|FlowControl|Sequence
    (:then_0
      (Audio|Components|Audio|FadeOut (Variables|Default|GetFireIgniteAudio) 0.1 0.0 "Linear" :!class /Script/Engine.AudioComponent))
    (:then_1
      (Audio|Components|Audio|FadeOut (Variables|Default|GetFireAudioLoop) 0.5 0.0 "Linear" :!class /Script/Engine.AudioComponent))))

(event Custom|TorchState
  (params (Enabled :type Boolean))
  ;[1 1 1] State
  (bind _output_get (Variables|Default|SetEnabled Enabled))
  (if _output_get
    (CallFunction|CheckInitialOverlappingActors)))

(event Custom|CheckInitialOverlappingActors
  ;[1 1 1] Check initial overlapping actors and handle Curie fire interactions
  (Networking|SwitchHasAuthority
    (:Authority
      (if (Curie|IsCurieActive)
        (for _array_element (Collision|GetOverlappingActors (Variables|Default|GetFire_CollisionVolume) 0 :!class /Script/Engine.PrimitiveComponent)
          (if (not (Utilities|Map|Contains (Variables|Default|GetInteract%20Handle%20Map) _array_element))
            (bind _returnvalue (Curie|ExecuteBeginFireInteract self self _array_element 1e+02 (Collision|MakeHitResult false false 0.0 0.0 (Transformation|GetActorLocation _array_element) (Transformation|GetActorLocation _array_element) "0,0,1" "0,0,1" 0 _array_element)))
            (Utilities|Map|Add (Variables|Default|GetInteract%20Handle%20Map) _array_element _returnvalue)))))))

(event OnComponentBeginOverlap(Fire_CollisionVolume)
  (params (OverlappedComponent :type Primitive%20Component%20Object%20Reference) (OtherActor :type Actor%20Object%20Reference) (OtherComp :type Primitive%20Component%20Object%20Reference) (OtherBodyIndex :type Integer) (bFromSweep :type Boolean) (SweepResult :type Hit%20Result%20Structure%20%28by%20ref%29))
  ;[1 1 1] On Fire Collision Begin Overlap
  (Networking|SwitchHasAuthority
    (:Authority
      (if (Curie|IsCurieActive)
        (if (not (Utilities|Map|Contains (Variables|Default|GetInteract%20Handle%20Map) OtherActor))
          (bind _returnvalue (Curie|ExecuteBeginFireInteract self self OtherActor 1e+02 SweepResult))
          (Utilities|Map|Add (Variables|Default|GetInteract%20Handle%20Map) OtherActor _returnvalue))))))

(event OnComponentEndOverlap(Fire_CollisionVolume)
  (params (OverlappedComponent :type Primitive%20Component%20Object%20Reference) (OtherActor :type Actor%20Object%20Reference) (OtherComp :type Primitive%20Component%20Object%20Reference) (OtherBodyIndex :type Integer))
  (bind _interact_handle_map (Variables|Default|GetInteract%20Handle%20Map))
  (bind (_value _returnvalue) (Utilities|Map|Find _interact_handle_map OtherActor))
  ;[1 1 1] On Fire Collision End Overlap
  (Networking|SwitchHasAuthority
    (:Authority
      (if (Curie|IsCurieActive)
        (if _returnvalue
          (Curie|ExecuteEndFireInteract OtherActor _value)
          (Utilities|Map|Remove (Variables|Default|GetInteract%20Handle%20Map) OtherActor))))))

(event EventDestroyed
  (bind _interact_handle_map (Variables|Default|GetInteract%20Handle%20Map))
  ;[1 1 1] Remove all fire created by this guy if destroyed.  Otherwise, people will burn eternally.
  (|CallParentFunction)
  (bind _keys (Utilities|Map|Keys (Variables|Default|GetInteract%20Handle%20Map)))
  (for _array_element _keys
    (if (.ReturnValue (Utilities|Map|Find _interact_handle_map _array_element))
      (Curie|ExecuteEndFireInteract _array_element (Utilities|Map|Find _interact_handle_map _array_element))))
  (Utilities|Map|Clear (Variables|Default|GetInteract%20Handle%20Map)))

(fn OnRep_Enabled
  (bind _ns_fire (Variables|Default|GetNS_Fire))
  (bind _fireigniteaudio (Variables|Default|GetFireIgniteAudio))
  (bind _fireigniteaudio_2 (Variables|Default|GetFireIgniteAudio))
  (bind _ns_fire_2 (Variables|Default|GetNS_Fire))
  (if (Variables|Default|GetEnabled)
    (Components|Activation|SetAutoActivate _ns_fire_2 true)
    (Components|Activation|Activate _ns_fire_2 true)
    (Audio|Components|Audio|Play (Variables|Default|GetFireAudioLoop) 0.0)
    (Collision|SetCollisionEnabled (Variables|Default|GetFire_CollisionVolume) "QueryOnly")
    (Rendering|Lighting|SetAttenuationRadius (Variables|Default|GetPointLight) 4.5e+02)
    (Audio|Components|Audio|SetSound _fireigniteaudio_2 "/Game/Sounds/Creative/Gadgets/Torch/Torch_Fire_Start_Cue.Torch_Fire_Start_Cue")
    (Audio|Components|Audio|Play _fireigniteaudio_2 0.0)
    (else
      (Components|Activation|SetAutoActivate _ns_fire false)
      (Components|Activation|Deactivate _ns_fire)
      (Audio|Components|Audio|FadeOut (Variables|Default|GetFireAudioLoop) 0.5 0.0 "Linear" :!class /Script/Engine.AudioComponent)
      (Collision|SetCollisionEnabled (Variables|Default|GetFire_CollisionVolume) "NoCollision")
      (Rendering|Lighting|SetAttenuationRadius (Variables|Default|GetPointLight) 0.0)
      (Audio|Components|Audio|SetSound _fireigniteaudio "/Game/Sounds/Creative/Gadgets/Torch/Torch_Fire_Stop_Cue.Torch_Fire_Stop_Cue")
      (Audio|Components|Audio|Play _fireigniteaudio 0.0))))

(fn ConstructionScript
  (|CallParentFunction)
  (Variables|Default|SetEnabled (Variables|Default|GetEnabled)))
